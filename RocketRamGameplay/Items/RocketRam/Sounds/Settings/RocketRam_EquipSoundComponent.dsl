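; Sound component for the Rocket Ram weapon that gates equip audio on whether the item is ready to use.
; On BeginPlay it casts its owner to the weapon, then caches that weapon's owning Athena player pawn.
; When audio is requested, it runs the parent playback only if the pawn has the Granted.Athena.RocketRam.ReadyToUse gameplay tag, and stops the audio otherwise.
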(event EventBeginPlay
  (|CallParentFunction)
  (bind _asb_weapon_rocket_ram (Utilities|Casting|CastToB_Weapon_RocketRam (Components|GetOwner)))
  (Utilities|IsValid _asb_weapon_rocket_ram
    (:Is%20Valid
      (bind _asplayer_pawn_athena (Utilities|Casting|CastToPlayerPawn_Athena (Actor|GetOwner _asb_weapon_rocket_ram)))
      (Variables|Default|SetPlayerRef _asplayer_pawn_athena))))

(event EventPlayAudio
  (bind _playerref (Variables|Default|GetPlayerRef)
    (:then
      (if (GameplayTags|HasMatchingGameplayTag _playerref "(TagName=\"Granted.Athena.RocketRam.ReadyToUse\")")
        (|CallParentFunction)
        (else
          (CallFunction|StopAudio))))
    (:else)))
